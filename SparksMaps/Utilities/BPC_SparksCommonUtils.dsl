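; Utility Blueprint component that toggles the Sparks capture tool widget, held as the Show Manager.
; On first show it creates the widget from a configurable class, adds it to the viewport and flags it as created, so later toggles only change visibility between Visible and Collapsed.
; A SetInputMode helper switches the player controller to UI-only input with a mouse cursor when shown, and game-only without one when hidden.
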(event EventShowSparksCaptureTool
  (params (bShow :type Boolean))
  ;[0.01161 0.01938 0.02217] Sparks.ToggleCaptureTool
  (if bShow
    (if (Variables|Default|GetbWidgetCreated)
      ;;[0.01161 0.01938 0.02217] Stage Manager is only created once. Once it is, just toggle visibility
      ;;; ShowShowManager
      (Utilities|IsValid (Variables|References|GetShowManager)
        (:Is%20Valid
          (Widget|SetVisibility (Variables|References|GetShowManager) "Visible")
          (CallFunction|SetInputMode false true)))
      (else
        ;;; CreateShowManager
        (bind _returnvalue (UserInterface|CreateWidget (Variables|SparksCommonUtilities|GetCaptureToolWidgetClass)))
        (bind _output_get (Variables|References|SetShowManager _returnvalue))
        (UserInterface|Viewport|AddtoViewport _output_get 0)
        (Variables|Default|SetbWidgetCreated true)
        (CallFunction|SetInputMode false true)))
    (else
      ;; HideShowManager
      (Utilities|IsValid (Variables|References|GetShowManager)
        (:Is%20Valid
          (Widget|SetVisibility (Variables|References|GetShowManager) "Collapsed")
          (CallFunction|SetInputMode true))))))

(fn SetInputMode
  (params (bGameplay :type Boolean) (bUI :type Boolean))
  (if bGameplay
    ;[0.01764 0.01764 0.01764] Game Mode & No Mouse Cursor
    (Input|SetInputModeGameOnly (Game|GetPlayerController 0))
    (Variables|/Script/Engine.PlayerController|SetbShowMouseCursor false (Game|GetPlayerController 0))
    (elif bUI
      ;[0.01764 0.01764 0.01764] UI & Mouse Cursor
      (Input|SetInputModeUIOnly (Game|GetPlayerController 0) (Variables|References|GetShowManager))
      (Variables|/Script/Engine.PlayerController|SetbShowMouseCursor true (Game|GetPlayerController 0)))))
