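# WBP_SparksCaptureWidgetError is a UMG error panel titled "Show Manager" that reports a missing capture tool widget class, apparently as a placeholder for that widget.
# Its text says to set CaptureToolWidgetClass on the SparksCommonUtils component in the project's Playspace blueprint.
# A left-anchored, vertically centered Overlay stacks a translucent rounded background, slight blur, and vertical box holding heading and message.
# It is purely static, with no variables or graph logic.
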WidgetBlueprint /SparksMaps/Utilities/WBP_SparksCaptureWidgetError
graphs [EventGraph {links [], nodes []}]
variables []
widgets {
 info {inheritedWidgetCount=0, parentClass=&UserWidget, rootWidgetClass=&CanvasPanel, widgetCount=9}
 widgets [
  Screen {widgetClassPath=&CanvasPanel, widget=Screen}
  Backing {
   widgetClassPath=&Overlay
   parent=Screen
   widget=Backing
   slot {class=CanvasPanelSlot, properties {layoutData {alignment=<0 0.5>, anchors {maximum=<0 0.5>, minimum=<0 0.5>}, offsets=<ltrb 75 0 500 700>}}}
  }
  Image_Background {
   widgetClassPath=&Border
   parent=Backing
   widget=Image_Background
   properties {
    background {
     bIsDynamicallyLoaded=false
     drawAs=RoundedBox
     imageType=NoImage
     mirroring=NoMirror
     resourceName=None
     resourceObject=null
     tiling=NoTile
     imageSize=<32 32>
     margin=<ltrb 0 0 0 0>
     outlineSettings {
      bUseBrushTransparency=false
      roundingType=FixedRadius
      width=0
      color {colorUseRule=UseColor_Specified, specifiedColor=<rgba 0 0 0 0>}
      cornerRadii=<25 25 25 25>
     }
     tintColor {colorUseRule=UseColor_Specified, specifiedColor=<rgba 0.0409 0.0168 0.1683 0.92>}
     uVRegion {bIsValid=false, max=<0 0>, min=<0 0>}
    }
    padding=<ltrb 14 14 14 14>
   }
   slot {class=OverlaySlot, properties {horizontalAlignment=HAlign_Fill, verticalAlignment=VAlign_Fill}}
  }
  Blur_Background {
   widgetClassPath=&BackgroundBlur
   parent=Backing
   widget=Blur_Background
   properties {blurStrength=0.6}
   slot {class=OverlaySlot, properties {horizontalAlignment=HAlign_Fill, verticalAlignment=VAlign_Fill}}
  }
  VerticalBox {
   widgetClassPath=&VerticalBox
   parent=Backing
   widget=VerticalBox
   slot {class=OverlaySlot, properties {horizontalAlignment=HAlign_Fill, verticalAlignment=VAlign_Fill}}
  }
  Heading {
   widgetClassPath=&VerticalBox
   parent=VerticalBox
   widget=Heading
   slot {class=VerticalBoxSlot, properties {verticalAlignment=VAlign_Top, padding=<ltrb 0 20 0 0>}}
  }
  Common_Text_Heading {
   widgetClassPath=&CommonTextBlock
   parent=Heading
   widget=Common_Text_Heading
   properties {
    justification=Center
    text=Show Manager
    textTransformPolicy=ToUpper
    visibility=HitTestInvisible
    colorAndOpacity {colorUseRule=UseColor_Specified, specifiedColor=<rgba 0.5 0.8333 1 1>}
    font {
     bForceMonospaced=false
     bMaterialIsStencil=false
     fontMaterial=null
     fontObject=&HeadingNow
     letterSpacing=0
     monospacedWidth=1
     size=30
     skewAmount=0
     typefaceFontName=76 Bold Italic
     outlineSettings {
      bApplyOutlineToDropShadows=false
      bMiteredCorners=false
      bSeparateFillAlpha=false
      outlineMaterial=null
      outlineSize=0
      outlineColor=<rgba 0 0 0 1>
     }
    }
    shadowOffset=<0 0>
   }
   slot {class=VerticalBoxSlot, properties {horizontalAlignment=HAlign_Center, verticalAlignment=VAlign_Center}}
  }
  CanvasPanel_0 {
   widgetClassPath=&CanvasPanel
   parent=VerticalBox
   widget=CanvasPanel_0
   slot {class=VerticalBoxSlot, properties {size {sizeRule=Fill, value=1}}}
  }
  CommonTextBlock_0 {
   widgetClassPath=&CommonTextBlock
   parent=CanvasPanel_0
   widget=CommonTextBlock_0
   properties {
    autoWrapText=true
    justification=Center
    style=&TextStyle-Base_C
    text="Missing widget class.\r\n\r\nGo to your project's implementation, and find the Playspace blueprint that it uses.\r\n\r\nSet the 'CaptureToolWidgetClass' on the SparksCommonUtils playspace component to the widget that you would like to create here."
    font {
     bForceMonospaced=false
     bMaterialIsStencil=false
     fontMaterial=null
     fontObject=&NotoSans
     letterSpacing=0
     monospacedWidth=1
     size=24
     skewAmount=0
     typefaceFontName=NotoSans-Regular
     outlineSettings {
      bApplyOutlineToDropShadows=false
      bMiteredCorners=false
      bSeparateFillAlpha=false
      outlineMaterial=null
      outlineSize=0
      outlineColor=<rgba 0 0 0 1>
     }
    }
    shadowOffset=<0 0>
   }
   slot {
    class=CanvasPanelSlot
    properties {layoutData {alignment=<0 0>, anchors {maximum=<1 1>, minimum=<0 0>}, offsets=<ltrb 14.7628 13.6777 13.9189 11.7117>}}
   }
  }
 ]
}
